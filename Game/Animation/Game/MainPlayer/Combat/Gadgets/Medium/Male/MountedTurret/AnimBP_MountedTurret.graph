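# Animation Blueprint for the mounted turret gun mesh, posing it as empty or manned and following the aim.
# Locomotion switches on bIsUsingMountedTurret between an empty pose and a pitch aim offset, then feeds DefaultSlot.
# A PedalScaler-timed pedal additive layers on top, and base_gear's world-space rotation is replaced with AimingYaw.
# The update event derives TurnRate from yaw change while manned; Locomotion_OLD is left unconnected.
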AnimBlueprint /Game/Animation/Game/MainPlayer/Combat/Gadgets/Medium/Male/MountedTurret/AnimBP_MountedTurret
target_skeleton=/Game/Gadgets/Assets/Turret_Gun/Meshes/Turret_Gun_Skeleton
graphs [
 AnimGraph {nodes [
   n1 = "Sequence Evaluator (by time) 'MT_Pedal_Right_CMM_Weapon'"!(ExplicitTime:float=Get PedalScaler() -> (PedalScaler:float)) -> (Pose:Pose Link Structure) {
    is=title AnimGraphNode_SequenceEvaluator
    sequence=&/MT_Pedal_Right_CMM_Weapon
   }
   AnimGraphNode_StateMachine_1 = Locomotion() {is=title AnimGraphNode_StateMachine, bReinitializeOnBecomingRelevant=false, bSkipFirstUpdateTransition=false}
   n2 = Slot 'DefaultSlot'(Source=@AnimGraphNode_StateMachine_1)
   n3 = "Transform (Modify) Bone - Bone: base_gear"!(
    ComponentPose=Local To Component(LocalPose=Apply Additive(Base=@n2, Additive=@n1) {alphaBoolBlend {blendOption=Linear}})
    Alpha=1
    Rotation_Roll:float=0
    Rotation_Pitch:float=0
    Rotation_Yaw:float=NormalizeAxis!(Angle:float=Get AimingYaw() -> (AimingYaw:float)) -> (ReturnValue:float)
   ) -> (Pose) {rotationMode=BMM_Replace, rotationSpace=BCS_WorldSpace, alphaBoolBlend {blendOption=Linear}, boneToModify {boneName=base_gear}}
   Output Pose(Result=Component To Local(ComponentPose=@n3)) {layerGroup=None}
   AnimGraphNode_StateMachine_0 = Locomotion_OLD() {is=title AnimGraphNode_StateMachine, bReinitializeOnBecomingRelevant=false, bSkipFirstUpdateTransition=false}
  ]}
 Locomotion_OLD {
  entry=Initialize
  path=AnimGraph.AnimGraphNode_StateMachine_0.Locomotion_OLD
  ids=id
  nodes [
   AnimStateNode_0 = Idle()
   AnimStateNode_1 = Rotate Turret() {is=title AnimStateNode}
   AnimStateNode_2 = TurretSeatEmpty() {is=title AnimStateNode}
   AnimStateNode_3 = EnterTurret() {is=title AnimStateNode}
   AnimStateNode_4 = SittingInTurret() {is=title AnimStateNode}
   AnimStateNode_5 = ExitTurret() {is=title AnimStateNode}
   AnimStateNode_6 = Initialize() {is=title AnimStateNode}
   AnimStateTransitionNode_0 = AnimStateTransitionNode()
   AnimStateTransitionNode_1 = AnimStateTransitionNode()
   AnimStateTransitionNode_2 = AnimStateTransitionNode()
   AnimStateTransitionNode_3 = AnimStateTransitionNode()
   AnimStateTransitionNode_4 = AnimStateTransitionNode() {bAutomaticRuleBasedOnSequencePlayerInState=true, crossfadeDuration=0.1}
   AnimStateTransitionNode_5 = AnimStateTransitionNode() {bAutomaticRuleBasedOnSequencePlayerInState=true, crossfadeDuration=0.1}
   AnimStateTransitionNode_6 = AnimStateTransitionNode() {crossfadeDuration=0.1}
   AnimStateTransitionNode_7 = AnimStateTransitionNode() {bAutomaticRuleBasedOnSequencePlayerInState=true, crossfadeDuration=0.1}
  ]
  links [
   {
    dst=SittingInTurret
    rule=AnimGraph.AnimGraphNode_StateMachine_0.Locomotion_OLD.AnimStateTransitionNode_4.Transition
    src=EnterTurret
    via=AnimStateTransitionNode_4
   }
   {
    dst=TurretSeatEmpty
    rule=AnimGraph.AnimGraphNode_StateMachine_0.Locomotion_OLD.AnimStateTransitionNode_3.Transition
    src=EnterTurret
    via=AnimStateTransitionNode_3
   }
   {
    dst=TurretSeatEmpty
    rule=AnimGraph.AnimGraphNode_StateMachine_0.Locomotion_OLD.AnimStateTransitionNode_7.Transition
    src=ExitTurret
    via=AnimStateTransitionNode_7
   }
   {
    dst=Rotate Turret
    rule=AnimGraph.AnimGraphNode_StateMachine_0.Locomotion_OLD.AnimStateTransitionNode_1.Transition
    src=Idle
    via=AnimStateTransitionNode_1
   }
   {
    dst=TurretSeatEmpty
    rule=AnimGraph.AnimGraphNode_StateMachine_0.Locomotion_OLD.AnimStateTransitionNode_5.Transition
    src=Initialize
    via=AnimStateTransitionNode_5
   }
   {
    dst=Idle
    rule=AnimGraph.AnimGraphNode_StateMachine_0.Locomotion_OLD.AnimStateTransitionNode_2.Transition
    src=Rotate Turret
    via=AnimStateTransitionNode_2
   }
   {
    dst=ExitTurret
    rule=AnimGraph.AnimGraphNode_StateMachine_0.Locomotion_OLD.AnimStateTransitionNode_6.Transition
    src=SittingInTurret
    via=AnimStateTransitionNode_6
   }
   {
    dst=EnterTurret
    rule=AnimGraph.AnimGraphNode_StateMachine_0.Locomotion_OLD.AnimStateTransitionNode_0.Transition
    src=TurretSeatEmpty
    via=AnimStateTransitionNode_0
   }
  ]
 }
 Idle {
  path=AnimGraph.AnimGraphNode_StateMachine_0.Locomotion_OLD.AnimStateNode_0.Idle
  nodes [Output Animation Pose(Result=Sequence Player 'MT_BasePose_CMM_Weapon'() {sequence=&/MT_BasePose_CMM_Weapon})]
 }
 Rotate Turret {
  path=AnimGraph.AnimGraphNode_StateMachine_0.Locomotion_OLD.AnimStateNode_1.Rotate Turret
  nodes [
   n1 = Sequence Player 'MT_Pedal_Left_CMM_Weapon'!(PlayRate:float=1) -> (Pose) {
    groupName=Turning
    method=SyncGroup
    sequence=&/MT_Pedal_Left_CMM_Weapon
   }
   n2 = "Sequence Player (additive) 'MT_Pedal_Right_CMM_Weapon'"!(PlayRate:float=1) -> (Pose) {
    groupName=Turning
    method=SyncGroup
    sequence=&MT_Pedal_Right_CMM_Weapon
   }
   Output Animation Pose(Result=Blend Poses by bool(BlendPose_0=@n2, BlendPose_1=@n1) {blendType=Linear})
  ]
 }
 TurretSeatEmpty {
  path=AnimGraph.AnimGraphNode_StateMachine_0.Locomotion_OLD.AnimStateNode_2.TurretSeatEmpty
  nodes [Output Animation Pose(Result=Sequence Player 'MT_SeatEmpty_Weapon'() {sequence=&/MT_SeatEmpty_Weapon})]
 }
 EnterTurret {
  path=AnimGraph.AnimGraphNode_StateMachine_0.Locomotion_OLD.AnimStateNode_3.EnterTurret
  nodes [
   n1 = AimOffset Player 'MT_AO_Weapon'!(
    X:float=0
    Y:float=NormalizeAxis(Angle=Get AimingPitch() -> (AimingPitch:float))
    BasePose:Pose Link Structure=Sequence Player 'MT_Enter_CMM_Weapon'!(bLoopAnimation:bool=false) -> (Pose) {sequence=&/MT_Enter_CMM_Weapon}
    Alpha:float=1
   ) -> (Pose:Pose Link Structure) {blendSpace=&/MT_AO_Weapon, alphaBoolBlend {blendOption=Linear}}
   n2 = Get PedalScaler()
   Output Animation Pose(Result=Apply Additive(
     Base=@n1
     Additive="Sequence Evaluator (by time) 'MT_Pedal_Right_CMM_Weapon'"(ExplicitTime=@n2) {sequence=&MT_Pedal_Right_CMM_Weapon}
    ) {alphaBoolBlend {blendOption=Linear}})
  ]
 }
 SittingInTurret {
  path=AnimGraph.AnimGraphNode_StateMachine_0.Locomotion_OLD.AnimStateNode_4.SittingInTurret
  nodes [
   n1 = AimOffset Player 'MT_AO_Weapon'(
    X=0
    Y=NormalizeAxis(Angle=Get AimingPitch())
    BasePose=Sequence Player 'MT_BasePose_CMM_Weapon'() {sequence=&MT_BasePose_CMM_Weapon}
    Alpha=1
   ) {blendSpace=&MT_AO_Weapon, alphaBoolBlend {blendOption=Linear}}
   n2 = Get PedalScaler()
   Output Animation Pose(Result=Apply Additive(
     Base=@n1
     Additive="Sequence Evaluator (by time) 'MT_Pedal_Right_CMM_Weapon'"(ExplicitTime=@n2) {sequence=&MT_Pedal_Right_CMM_Weapon}
    ) {alphaBoolBlend {blendOption=Linear}})
  ]
 }
 ExitTurret {
  path=AnimGraph.AnimGraphNode_StateMachine_0.Locomotion_OLD.AnimStateNode_5.ExitTurret
  nodes [
   n1 = NormalizeAxis(Angle=Get AimingPitch())
   Output Animation Pose(Result=Sequence Player 'MT_Exit_CMM_Weapon'!(StartPosition:float=MapRangeClamped(Value=@n1, InRangeA=90, InRangeB=-30, OutRangeB=0.75), bLoopAnimation:bool=false) -> (Pose) {sequence=&/MT_Exit_CMM_Weapon})
  ]
 }
 Initialize {
  path=AnimGraph.AnimGraphNode_StateMachine_0.Locomotion_OLD.AnimStateNode_6.Initialize
  nodes [
   Sequence Player 'MT_Exit_CMM_Weapon'(StartPosition=0.5, bLoopAnimation=false) {sequence=&MT_Exit_CMM_Weapon}
   Output Animation Pose(Result=Sequence Player 'MT_Transform_Weapon'!(bLoopAnimation:bool=false) -> (Pose) {sequence=&/MT_Transform_Weapon})
  ]
 }
 Transition {
  path=AnimGraph.AnimGraphNode_StateMachine_0.Locomotion_OLD.AnimStateTransitionNode_0.Transition
  nodes [Result(bCanEnterTransition=Get bIsUsingMountedTurret() -> (bIsUsingMountedTurret:bool))]
 }
 Transition {
  path=AnimGraph.AnimGraphNode_StateMachine_0.Locomotion_OLD.AnimStateTransitionNode_1.Transition
  nodes [Result(bCanEnterTransition=Get isTurning() -> (isTurning:bool))]
 }
 Transition {
  path=AnimGraph.AnimGraphNode_StateMachine_0.Locomotion_OLD.AnimStateTransitionNode_2.Transition
  nodes [Result(bCanEnterTransition=NOT Boolean(A=Get isTurning()))]
 }
 Transition {
  path=AnimGraph.AnimGraphNode_StateMachine_0.Locomotion_OLD.AnimStateTransitionNode_3.Transition
  nodes [Result(bCanEnterTransition=NOT Boolean(A=Get bIsUsingMountedTurret()))]
 }
 Transition {
  path=AnimGraph.AnimGraphNode_StateMachine_0.Locomotion_OLD.AnimStateTransitionNode_4.Transition
  nodes [Result(bCanEnterTransition="float <= float"!(A:double="Get Relevant Anim Time Remaining (EnterTurret)"!() -> (ReturnValue:float) {is=title K2Node_AnimGetter}, B:double=0.1) -> (ReturnValue:bool) {is=title K2Node_CallFunction})]
 }
 Transition {
  path=AnimGraph.AnimGraphNode_StateMachine_0.Locomotion_OLD.AnimStateTransitionNode_5.Transition
  nodes [Result(bCanEnterTransition="float <= float"(A="Get Relevant Anim Time Remaining (Initialize)"!() -> (ReturnValue:float) {is=title K2Node_AnimGetter}, B=0.1))]
 }
 Transition {
  path=AnimGraph.AnimGraphNode_StateMachine_0.Locomotion_OLD.AnimStateTransitionNode_6.Transition
  nodes [Result(bCanEnterTransition=NOT Boolean(A=Get bIsUsingMountedTurret()))]
 }
 Transition {
  path=AnimGraph.AnimGraphNode_StateMachine_0.Locomotion_OLD.AnimStateTransitionNode_7.Transition
  nodes [Result(bCanEnterTransition="float <= float"(A="Get Relevant Anim Time Remaining (ExitTurret)"!() -> (ReturnValue:float) {is=title K2Node_AnimGetter}, B=0.1))]
 }
 Locomotion {
  entry=TurretEmptyPose
  path=AnimGraph.AnimGraphNode_StateMachine_1.Locomotion
  ids=id
  nodes [
   AnimStateNode_0 = Idle()
   AnimStateNode_1 = Rotate Turret()
   AnimStateNode_4 = SittingInTurret()
   AnimStateNode_7 = TurretEmptyPose() {is=title AnimStateNode}
   AnimStateTransitionNode_1 = AnimStateTransitionNode()
   AnimStateTransitionNode_2 = AnimStateTransitionNode()
   AnimStateTransitionNode_8 = AnimStateTransitionNode()
   AnimStateTransitionNode_9 = AnimStateTransitionNode()
  ]
  links [
   {
    dst=Rotate Turret
    rule=AnimGraph.AnimGraphNode_StateMachine_1.Locomotion.AnimStateTransitionNode_1.Transition
    src=Idle
    via=AnimStateTransitionNode_1
   }
   {
    dst=Idle
    rule=AnimGraph.AnimGraphNode_StateMachine_1.Locomotion.AnimStateTransitionNode_2.Transition
    src=Rotate Turret
    via=AnimStateTransitionNode_2
   }
   {
    dst=TurretEmptyPose
    rule=AnimGraph.AnimGraphNode_StateMachine_1.Locomotion.AnimStateTransitionNode_9.Transition
    src=SittingInTurret
    via=AnimStateTransitionNode_9
   }
   {
    dst=SittingInTurret
    rule=AnimGraph.AnimGraphNode_StateMachine_1.Locomotion.AnimStateTransitionNode_8.Transition
    src=TurretEmptyPose
    via=AnimStateTransitionNode_8
   }
  ]
 }
 Idle {
  path=AnimGraph.AnimGraphNode_StateMachine_1.Locomotion.AnimStateNode_0.Idle
  nodes [Output Animation Pose(Result=Sequence Player 'MT_BasePose_CMM_Weapon'() {sequence=&MT_BasePose_CMM_Weapon})]
 }
 Rotate Turret {
  path=AnimGraph.AnimGraphNode_StateMachine_1.Locomotion.AnimStateNode_1.Rotate Turret
  nodes [
   n1 = Sequence Player 'MT_Pedal_Left_CMM_Weapon'(PlayRate=1) {
    groupName=Turning
    method=SyncGroup
    sequence=&MT_Pedal_Left_CMM_Weapon
   }
   n2 = "Sequence Player (additive) 'MT_Pedal_Right_CMM_Weapon'"(PlayRate=1) {
    groupName=Turning
    method=SyncGroup
    sequence=&MT_Pedal_Right_CMM_Weapon
   }
   Output Animation Pose(Result=Blend Poses by bool(BlendPose_0=@n2, BlendPose_1=@n1) {blendType=Linear})
  ]
 }
 SittingInTurret {
  path=AnimGraph.AnimGraphNode_StateMachine_1.Locomotion.AnimStateNode_4.SittingInTurret
  nodes [
   n1 = AimOffset Player 'MT_AO_Weapon'(
    X=0
    Y=NormalizeAxis(Angle=Get AimingPitch())
    BasePose=Sequence Player 'MT_BasePose_CMM_Weapon'() {sequence=&MT_BasePose_CMM_Weapon}
    Alpha=1
   ) {blendSpace=&MT_AO_Weapon, alphaBoolBlend {blendOption=Linear}}
   Output Animation Pose(Result=@n1)
  ]
 }
 TurretEmptyPose {
  path=AnimGraph.AnimGraphNode_StateMachine_1.Locomotion.AnimStateNode_7.TurretEmptyPose
  nodes [Output Animation Pose(Result="Sequence Evaluator (by time) 'MT_Enter_CMM_Weapon'"!(ExplicitTime:float=0) -> (Pose:Pose Link Structure) {
     is=title AnimGraphNode_SequenceEvaluator
     sequence=&MT_Enter_CMM_Weapon
    })]
 }
 Transition {
  path=AnimGraph.AnimGraphNode_StateMachine_1.Locomotion.AnimStateTransitionNode_1.Transition
  nodes [Result(bCanEnterTransition=Get isTurning())]
 }
 Transition {
  path=AnimGraph.AnimGraphNode_StateMachine_1.Locomotion.AnimStateTransitionNode_2.Transition
  nodes [Result(bCanEnterTransition=NOT Boolean(A=Get isTurning()))]
 }
 Transition {
  path=AnimGraph.AnimGraphNode_StateMachine_1.Locomotion.AnimStateTransitionNode_8.Transition
  nodes [Result(bCanEnterTransition=Get bIsUsingMountedTurret())]
 }
 Transition {
  path=AnimGraph.AnimGraphNode_StateMachine_1.Locomotion.AnimStateTransitionNode_9.Transition
  nodes [Result(bCanEnterTransition=NOT Boolean(A=Get bIsUsingMountedTurret()))]
 }
 EventGraph {nodes [
   Event BlueprintUpdateAnimation()
   > Is Valid(InputObject=TryGetPawnOwner())
   > Sequence()
   n1 = Get AimingYaw()
   n2 = NormalizeAxis(Angle=float - float(A=NormalizeAxis(Angle=@n1), B=Get AimingYawLastFrame() -> (AimingYawLastFrame:double)) {is=title K2Node_CallFunction})
   n3 = "Max (Float)"(A=GetWorldDeltaSeconds!() -> (ReturnValue:double), B=0.016)
   > Set TurnRate(TurnRate:float=Select(Option 0:double=0, Option 1:double=float / float(A=@n2, B=@n3) {is=title K2Node_CallFunction}, Index=Get bIsUsingMountedTurret()) -> (ReturnValue:double)) -> (Output_Get:float)
   > Set AimingYawLastFrame(AimingYawLastFrame:double=NormalizeAxis(Angle=Get AimingYaw())) -> (Output_Get:double)
  ]}
]
variables [
 AimingYawLastFrame {type=number}
 isTurning {type=boolean}
]
