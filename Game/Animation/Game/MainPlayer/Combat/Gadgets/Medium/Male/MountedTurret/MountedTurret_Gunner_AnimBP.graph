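# Animation Blueprint that poses the player gunner on a mounted turret via the VehicleFinalPoseOverride layer.
# A turn/pitch-rate aim offset over the base pose receives PedalScaler-scrubbed pedal additives and montage slots.
# Leg IK nodes are repurposed to pull both hands onto ik_hand bones.
# Steering-wheel socket math lacks an event trigger; its Modify Bone nodes and the Locomotion state machine are disconnected from the output.
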AnimBlueprint /Game/Animation/Game/MainPlayer/Combat/Gadgets/Medium/Male/MountedTurret/MountedTurret_Gunner_AnimBP
target_skeleton=&Fortnite_M_Avg_Player_Skeleton
graphs [
 AnimGraph {nodes [Output Pose() {layerGroup=None}]}
 EventGraph {nodes [
   n1 = GetSocketTransform(
    self=Get SkeletalMesh!(self:Fort Athena Vehicle Object Reference=Get MountedTurret() -> (MountedTurret:Fort Mounted Turret Object Reference)) -> (SkeletalMesh:Skeletal Mesh Component Budgeted Object Reference)
    InSocketName=steering_wheel_hand_l
    TransformSpace=RTS_World
   )
   Set WorldPosLeft(WorldPosLeft:Vector=TransformLocation!(T:"Transform (by ref)"=@n1, Location:Vector=Get LocalLeftOffset() -> (LocalLeftOffset:Vector)) -> (ReturnValue:Vector)) -> (Output_Get:Vector)
   > Set WorldRotLeft(WorldRotLeft:Rotator=TransformRotation!(T:"Transform (by ref)"=@n1, Rotation:Rotator=Get LocalRotLeftOffset() -> (LocalRotLeftOffset:Rotator)) -> (ReturnValue:Rotator)) -> (Output_Get:Rotator)
   n2 = GetSocketTransform(self=Get SkeletalMesh(self=Get MountedTurret()), InSocketName=steering_wheel_hand_r, TransformSpace=RTS_World)
   > Set WorldPosRight(WorldPosRight:Vector=TransformLocation(T=@n2, Location=Get LocalRightOffset() -> (LocalRightOffset:Vector))) -> (Output_Get:Vector)
   > Set WorldRotRight(WorldRotRight:Rotator=TransformRotation(T=@n2, Rotation=Get LocalRotRightOffset() -> (LocalRotRightOffset:Rotator))) -> (Output_Get:Rotator)
  ]}
 VehicleFinalPoseOverride {nodes [
   Input Pose() {name=InFinalPose}
   n1 = "Transform (Modify) Bone - Bone: ik_hand_l"!(
    ComponentPose
    AlphaCurveName:Name=BothHandFK
    Translation=Get WorldPosLeft() -> (WorldPosLeft:Vector)
    Rotation=Get WorldRotLeft() -> (WorldRotLeft:Rotator)
   ) -> (Pose) {
    alphaInputType=Curve
    rotationMode=BMM_Replace
    rotationSpace=BCS_WorldSpace
    translationMode=BMM_Replace
    translationSpace=BCS_WorldSpace
    alphaBoolBlend {blendOption=Linear}
    alphaScaleBiasClamp {bias=1, scale=-1}
    boneToModify {boneName=ik_hand_l}
   }
   "Transform (Modify) Bone - Bone: ik_hand_r"!(
    ComponentPose=@n1
    AlphaCurveName:Name=BothHandFK
    Translation=Get WorldPosRight() -> (WorldPosRight:Vector)
    Rotation=Get WorldRotRight() -> (WorldRotRight:Rotator)
   ) -> (Pose) {
    alphaInputType=Curve
    rotationMode=BMM_Replace
    rotationSpace=BCS_WorldSpace
    translationMode=BMM_Replace
    translationSpace=BCS_WorldSpace
    alphaBoolBlend {blendOption=Linear}
    alphaScaleBiasClamp {bias=1, scale=-1}
    boneToModify {boneName=ik_hand_r}
   }
   n2 = "Sequence Evaluator (by time) 'MT_Pedal_Right_CMM'"!(ExplicitTime:float=Get PedalScaler() -> (PedalScaler:float)) -> (Pose:Pose Link Structure) {
    is=title AnimGraphNode_SequenceEvaluator
    sequence=&/MT_Pedal_Right_CMM
   }
   n3 = "Blend Poses (EFortCustomGender)"!(
    BlendPose_0:Pose Link Structure=Additive Identity Pose()
    BlendPose_1:Pose Link Structure=Additive Identity Pose()
    BlendTime_0:float=0.1
    BlendTime_1:float=0.1
    ActiveEnumValue:Byte=Get Gender() -> (Gender:EFortCustomGender Enum=Invalid)
   ) -> (Pose:Pose Link Structure) {
    is=title AnimGraphNode_BlendListByEnum
    blendType=Linear
    blendPose [{linkId=-1, sourceLinkId=-1}, {linkId=-1, sourceLinkId=-1}]
    blendTime [0.1,0.1]
   }
   n4 = AimOffset Player 'MT_AO_CMM'!(
    X:float=float * float(A=Get TurnRate() -> (TurnRate:float), B=0.25)
    Y:float=float * float(A=Get PitchRate() -> (PitchRate:float), B=0.5)
    BasePose:Pose Link Structure=Sequence Player 'MT_BasePose_CMM'() {sequence=&/MT_BasePose_CMM}
    bAlphaBoolEnabled:bool=true
   ) -> (Pose:Pose Link Structure) {
    alphaInputType=Bool
    blendSpace=&/MT_AO_CMM
    alphaBoolBlend {blendInTime=0.2, blendOption=Linear, blendOutTime=0.2}
   }
   n5 = Slot 'FullBody'(Source=Apply Additive(Base=@n4, Additive=@n3) {alphaBoolBlend {blendOption=Linear}}) {slotName=FullBody}
   n6 = Leg IK!(
    ComponentPose:Component Space Pose Link Structure=Local To Component(LocalPose=Apply Additive(Base=@n5, Additive=@n2) {alphaBoolBlend {blendOption=Linear}})
    Alpha:float=Get RightHandIKAlpha() -> (RightHandIKAlpha:float)
   ) -> (Pose:Component Space Pose Link Structure) {
    is=title AnimGraphNode_LegIK
    alphaBoolBlend {blendOption=Linear}
    legsDefinition [{
      bEnableKneeTwistCorrection=false
      bEnableRotationLimit=false
      footBoneForwardAxis=Y
      hingeRotationAxis=Z
      minRotationAngle=15
      numBonesInLimb=2
      twistOffsetCurveName=None
      fKFootBone {boneName=hand_r}
      iKFootBone {boneName=ik_hand_r}
     }]
   }
   n7 = Leg IK(ComponentPose=@n6, Alpha=Get LeftHandIKAlpha() -> (LeftHandIKAlpha:float)) {
    alphaBoolBlend {blendOption=Linear}
    legsDefinition [{
      bEnableKneeTwistCorrection=false
      bEnableRotationLimit=false
      footBoneForwardAxis=Y
      hingeRotationAxis=Z
      minRotationAngle=15
      numBonesInLimb=2
      twistOffsetCurveName=None
      fKFootBone {boneName=hand_l}
      iKFootBone {boneName=ik_hand_l}
     }]
   }
   Output Pose(Result=Slot 'FullBodyNoAdditive'(Source=Component To Local(ComponentPose=@n7)) {slotName=FullBodyNoAdditive}) {layerGroup=VehicleOverride}
   AnimGraphNode_StateMachine_0 = Locomotion() {is=title AnimGraphNode_StateMachine, bReinitializeOnBecomingRelevant=false, bSkipFirstUpdateTransition=false}
   Get AimPitch()
  ]}
 Locomotion {
  entry=InTurret
  path=VehicleFinalPoseOverride.AnimGraphNode_StateMachine_0.Locomotion
  ids=id
  nodes [
   AnimStateNode_0 = Idle()
   AnimStateNode_1 = EnterTurret() {is=title AnimStateNode}
   AnimStateNode_2 = Rotate Turret() {is=title AnimStateNode}
   AnimStateNode_3 = InTurret() {is=title AnimStateNode}
   AnimStateTransitionNode_1 = AnimStateTransitionNode() {crossfadeDuration=0.1}
   AnimStateTransitionNode_5 = AnimStateTransitionNode()
   AnimStateTransitionNode_9 = AnimStateTransitionNode()
  ]
  links [
   {
    dst=InTurret
    rule=VehicleFinalPoseOverride.AnimGraphNode_StateMachine_0.Locomotion.AnimStateTransitionNode_1.Transition
    src=EnterTurret
    via=AnimStateTransitionNode_1
   }
   {
    dst=Rotate Turret
    rule=VehicleFinalPoseOverride.AnimGraphNode_StateMachine_0.Locomotion.AnimStateTransitionNode_5.Transition
    src=Idle
    via=AnimStateTransitionNode_5
   }
   {
    dst=Idle
    rule=VehicleFinalPoseOverride.AnimGraphNode_StateMachine_0.Locomotion.AnimStateTransitionNode_9.Transition
    src=Rotate Turret
    via=AnimStateTransitionNode_9
   }
  ]
 }
 Idle {
  path=VehicleFinalPoseOverride.AnimGraphNode_StateMachine_0.Locomotion.AnimStateNode_0.Idle
  nodes [Output Animation Pose(Result=Sequence Player 'MT_BasePose_CMM'() {sequence=&MT_BasePose_CMM})]
 }
 EnterTurret {
  path=VehicleFinalPoseOverride.AnimGraphNode_StateMachine_0.Locomotion.AnimStateNode_1.EnterTurret
  nodes [
   n1 = "Blend Poses (EFortCustomGender)"(
    BlendPose_0=Sequence Player 'MT_Enter_CMM'!(bLoopAnimation:bool=false) -> (Pose) {sequence=&/MT_Enter_CMM}
    BlendPose_1=Sequence Player 'MT_Enter_CMF'!(bLoopAnimation:bool=false) -> (Pose) {sequence=/Game/Animation/Game/MainPlayer/Combat/Gadgets/Medium/Female/MountedTurret/MT_Enter_CMF}
    BlendTime_0=0.1
    BlendTime_1=0.1
    ActiveEnumValue=Get Gender() -> (Gender=Invalid)
   ) {blendType=Linear, blendPose [{linkId=-1, sourceLinkId=-1}, {linkId=-1, sourceLinkId=-1}], blendTime [0.1,0.1]}
   n2 = Get PedalScaler()
   Output Animation Pose(Result=Apply Additive(
     Base=@n1
     Additive="Sequence Evaluator (by time) 'MT_Pedal_Right_CMM'"(ExplicitTime=@n2) {sequence=&MT_Pedal_Right_CMM}
    ) {alphaBoolBlend {blendOption=Linear}})
  ]
 }
 Rotate Turret {
  path=VehicleFinalPoseOverride.AnimGraphNode_StateMachine_0.Locomotion.AnimStateNode_2.Rotate Turret
  nodes [Output Animation Pose(Result=Blend Poses by bool(
     BlendPose_0="Sequence Player (additive) 'MT_Pedal_Right_CMM'"!(PlayRate:float=1) -> (Pose) {sequence=&MT_Pedal_Right_CMM}
     BlendPose_1=Sequence Player 'MT_Pedal_Left_CMM'!(PlayRate:float=1) -> (Pose) {sequence=&/MT_Pedal_Left_CMM}
    ) {blendType=Linear})]
 }
 InTurret {
  path=VehicleFinalPoseOverride.AnimGraphNode_StateMachine_0.Locomotion.AnimStateNode_3.InTurret
  nodes [
   n1 = Get PedalScaler()
   Output Animation Pose(Result=Apply Additive(
     Base=Sequence Player 'MT_BasePose_CMM'() {sequence=&MT_BasePose_CMM}
     Additive="Sequence Evaluator (by time) 'MT_Pedal_Right_CMM'"(ExplicitTime=@n1) {sequence=&MT_Pedal_Right_CMM}
    ) {alphaBoolBlend {blendOption=Linear}})
  ]
 }
 Transition {
  path=VehicleFinalPoseOverride.AnimGraphNode_StateMachine_0.Locomotion.AnimStateTransitionNode_1.Transition
  nodes [Result(bCanEnterTransition="float <= float"!(A:double="Get Relevant Anim Time Remaining (EnterTurret)"!() -> (ReturnValue:float) {is=title K2Node_AnimGetter}, B:double=0.1) -> (ReturnValue:bool) {is=title K2Node_CallFunction})]
 }
 Transition {
  path=VehicleFinalPoseOverride.AnimGraphNode_StateMachine_0.Locomotion.AnimStateTransitionNode_5.Transition
  nodes [Result(bCanEnterTransition=Get isTurning() -> (isTurning:bool))]
 }
 Transition {
  path=VehicleFinalPoseOverride.AnimGraphNode_StateMachine_0.Locomotion.AnimStateTransitionNode_9.Transition
  nodes [Result(bCanEnterTransition=NOT Boolean(A=Get isTurning()))]
 }
 VehicleFullBodyOverride {nodes [
   Input Pose() {name=InPoseFullBody}
   Output Pose() {layerGroup=VehicleOverride}
  ]}
 VehicleLowerBodyOverride {nodes [
   Input Pose() {name=InPoseLowerBodyDefault}
   Input Pose() {name=InPoseUpperBody}
   Output Pose() {layerGroup=VehicleOverride}
  ]}
 VehicleSplitBodyOverride {nodes [
   Input Pose() {name=InPoseSplitBody}
   Input Pose() {name=InPoseUpperAndLowerBody}
   Input Pose() {name=InPoseUpperAndLowerBodyRemoveRoot}
   Output Pose() {layerGroup=Vehicle Override}
  ]}
]
variables [
 isTurning {type=boolean}
 LocalLeftOffset {type=Vector, default=<-3.5 -0.5 -5>}
 LocalRightOffset {type=Vector, default=<3.5 -0.5 5>}
 LocalRotLeftOffset {type=Rotator}
 LocalRotRightOffset {type=Rotator}
 WorldPosLeft {type=Vector}
 WorldPosRight {type=Vector}
 WorldRotLeft {type=Rotator}
 WorldRotRight {type=Rotator}
]
